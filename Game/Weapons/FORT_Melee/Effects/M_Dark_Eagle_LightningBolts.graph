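# Unlit additive particle sprite material for animated lightning bolts in the Dark Eagle melee weapon effects.
# A time-driven 2x2 arc flipbook, tinted by particle color, blends core and glow channels with optional flashing from MF_Lightning_Flashes.
# A switchable world-space sphere mask around the object distorts flipbook UVs and shapes opacity, while a radial UV mask erodes the center.
# Dynamic parameters drive flicker, glow, erosion, and an Offset shared by timing and sphere radius.
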Material /Game/Weapons/FORT_Melee/Effects/M_Dark_Eagle_LightningBolts
graph {
 properties {
  bUsedWithNiagaraSprites=true
  bUsedWithParticleSprites=true
  blendMode=BLEND_Additive
  pixelDepthOffsetMode=PDOM_Legacy
  shadingModel=MSM_Unlit
  usedShadingModels=Unlit
 }
 nodes [
  n1 = DynamicParameter() {defaultValue=<rgba 0 1 0.15 0>, paramNames [LightningFlicker, Offset, GlowAmount, Erode]}
  n2 = LinearInterpolate(
   A=Constant() {r=1}
   B=MaterialFunctionCall!(TimeMult=@n1.Offset) {materialFunction=/Game/Effects/Fort_Effects/Materials/_Functions/MF_Lightning_Flashes}.Result
   Alpha=@n1.LightningFlicker
  )
  LinearInterpolate(A=Subtract(A=ParticlePositionWS(), B=WorldPosition().XYZ), Alpha=Clamp(None=@n2))
  MaterialFunctionCall() {materialFunction=&DebugTimeSine}
  MaterialFunctionCall() {materialFunction=&DebugTimeSine}
  n3 = ParticleColor()
  n4 = Multiply(A=Add(A=Time(), B=@n1.Offset), B=ScalarParameter() {defaultValue=3, parameterName=FlipbookSpeed})
  n5 = Constant() {r=2}
  n6 = ScalarParameter() {defaultValue=10, parameterName=SphereSize}
  n7 = MaterialFunctionCall() {materialFunction=/Engine/Functions/Engine_MaterialFunctions02/WorldPositionOffset/ObjectScale}
  n8 = SphereMask(A=ObjectPositionWS(), B=WorldPosition().XYZ, Radius=Multiply(A=@n1.Offset, B=Multiply(A=@n6, B=@n7.Scale X))) {attenuationRadius=0.5, hardnessPercent=80}
  n9 = Multiply(A=@n8, B=Multiply(A=OneMinus(None=@n8)) {constB=15})
  n10 = Multiply(A=@n9) {constB=0.02}
  n11 = Constant()
  n12 = TextureSample(UVs=Add(
    A=MaterialFunctionCall("Animation  Phase (0-1)"=@n4, Number of Rows=@n5, Number of Columns=@n5) {materialFunction=&FlipBook}.UVs
    B=StaticSwitchParameter(True=@n10, False=@n11) {defaultValue=true, parameterName=SphereMaskEdges}
   )) {texture=/Game/Effects/Fort_Effects/Textures/T_Elec_CentralArcs_Flipbook}
  n13 = Multiply(A=@n12.R, B=@n1.GlowAmount) {constB=0.15}
  n14 = Multiply(A=@n3.RGB, B=Multiply(A=Add(A=@n13, B=@n12.G), B=Multiply(A=@n2) {constB=1.5}))
  MaterialFunctionCall(Emissive=MaterialFunctionCall(LDR Input=Clamp(None=@n14), HDR Tint=@n3.RGB) {materialFunction=&DeriveHDRfromLDR}.RemappedRange) {materialFunction=&MF_FX_EmissiveAdjustments}
  n15 = SphereMask(A=TextureCoordinate(), B=Constant() {r=0.5}, Radius=@n1.Erode, Hardness=LinearInterpolate(Alpha=@n1.Erode) {constB=0.6}) {hardnessPercent=50}
  n16 = Add(A=@n8, B=@n9)
  n17 = Constant() {r=1}
  Multiply(
   A=@n3.A
   B=Multiply(A=OneMinus(None=@n15), B=StaticSwitchParameter(True=@n16, False=@n17) {defaultValue=true, parameterName=SphereMaskEdges})
  )
 ]
}
