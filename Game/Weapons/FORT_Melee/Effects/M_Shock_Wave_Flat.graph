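# Translucent, unlit, two-sided particle material, named as a flat shock wave effect for melee weapons.
# Particle color RGB feeds the emissive adjustments function.
# A smoke shoot-out ramp, offset along V by dynamic parameter 1, is multiplied by a soft vertical gradient scaled by dynamic parameter 2 to form a mask.
# The particle system thus drives the ramp's movement and the mask's strength, with no time-based panning.
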Material /Game/Weapons/FORT_Melee/Effects/M_Shock_Wave_Flat
graph {
 properties {blendMode=BLEND_Translucent, pixelDepthOffsetMode=PDOM_Legacy, shadingModel=MSM_Unlit, twoSided=true, usedShadingModels=Unlit}
 nodes [
  Constant() {r=0.5}
  Constant() {r=1}
  Constant() {r=10}
  MaterialFunctionCall(Emissive=ParticleColor().RGB) {materialFunction=&MF_FX_EmissiveAdjustments}
  n1 = TextureCoordinate()
  n2 = DynamicParameter() {defaultValue=<rgba 0 0 0 1>}
  n3 = AppendVector(A=Constant(), B=@n2.Param1)
  Multiply(
   A=TextureSample(UVs=Add(A=@n1, B=@n3)) {texture=/Game/Effects/Fort_Effects/Textures/Ramps/T_Smoke_Shoot_Out_Ramp}.G
   B=Multiply(A=TextureSample() {texture=/Game/Global/Textures/Default/Masks_and_Gradients/T_Soft_Ramp_Vert_03}.G, B=@n2.Param2)
  )
 ]
}
