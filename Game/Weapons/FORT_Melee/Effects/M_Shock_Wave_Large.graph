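# Unlit, translucent, two-sided material for a large melee shock wave effect, enabled for mesh particles.
# Two oppositely drifting samples of a tiling smoke texture multiply into a subtle UV distortion of the Super_Melee_Wave texture.
# The partly desaturated wave is tinted by particle color and brightened through MF_FX_EmissiveAdjustments, and its alpha is multiplied by particle alpha.
# A second, similar distortion chain and a tiled texture coordinate feed nothing.
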Material /Game/Weapons/FORT_Melee/Effects/M_Shock_Wave_Large
graph {
 properties {
  bUsedWithMeshParticles=true
  blendMode=BLEND_Translucent
  pixelDepthOffsetMode=PDOM_Legacy
  shadingModel=MSM_Unlit
  twoSided=true
  usedShadingModels=Unlit
 }
 nodes [
  n1 = Constant2Vector() {g=3, r=3}
  n2 = TextureCoordinate()
  n3 = Multiply(
   A=TextureSample(UVs=Panner(Coordinate=@n2) {speedX=0.1, speedY=0.4}) {samplerType=SAMPLERTYPE_LinearColor, texture=/Game/Global/Textures/Default/Tileable_Patterns/T_Tile_Smoke_Small}.G
   B=TextureSample(UVs=Panner(Coordinate=@n2) {speedX=-0.1, speedY=0.3}) {samplerType=SAMPLERTYPE_LinearColor, texture=&T_Tile_Smoke_Small}.G
  )
  Add(A=TextureCoordinate(), B=Multiply(A=Multiply(A=@n1, B=@n3), B=Constant() {r=0.01}))
  Constant() {r=20}
  n4 = TextureCoordinate()
  n5 = Panner(Coordinate=@n4) {speedX=0.1, speedY=0.6}
  n6 = Panner(Coordinate=@n4) {speedX=-0.2, speedY=0.5}
  n7 = Multiply(
   A=Constant2Vector() {g=3, r=3}
   B=Multiply(
    A=TextureSample(UVs=@n5) {samplerType=SAMPLERTYPE_LinearColor, texture=&T_Tile_Smoke_Small}.G
    B=TextureSample(UVs=@n6) {samplerType=SAMPLERTYPE_LinearColor, texture=&T_Tile_Smoke_Small}.G
   )
  )
  n8 = Constant() {r=0.01}
  n9 = TextureSample(UVs=Add(A=TextureCoordinate(), B=Multiply(A=@n7, B=@n8))) {texture=/Game/Effects/Fort_Effects/Textures/Melee/Super_Melee_Wave}
  n10 = Desaturation(None=@n9.RGB, Fraction=Constant() {r=0.6})
  n11 = ParticleColor()
  MaterialFunctionCall(Emissive=Multiply(A=Constant() {r=20}, B=Multiply(A=@n10, B=@n11.RGB))) {materialFunction=&MF_FX_EmissiveAdjustments}
  Multiply(A=@n9.A, B=@n11.A)
  TextureCoordinate() {uTiling=5, vTiling=5}
 ]
}
