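# Translucent, unlit mesh-particle material for the head of the electrical pickaxe's lightning bolt effect.
# World position offset combines SplineThicken, time-scrolled vector noise curl (zero on mobile) and a sphere-masked offset scaled by dynamic parameters.
# Emissive blends tip, back and middle colors along UVs, scaled by particle green, through MF_FX_EmissiveAdjustments.
# Opacity multiplies a CenterGrad-shifted sphere mask, vertex green, OpacityPanner and optional non-mobile depth fade.
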Material /Game/Weapons/FORT_Melee/Effects/Materials/M_Electrical_PickAxe_LightningBolt_Head_01
graph {
 properties {
  bUsedWithMeshParticles=true
  bUsedWithNiagaraMeshParticles=true
  blendMode=BLEND_Translucent
  opacityMaskClipValue=0.1
  pixelDepthOffsetMode=PDOM_Legacy
  shadingModel=MSM_Unlit
  twoSided=true
  usedShadingModels=Unlit
 }
 nodes [
  n1 = ScalarParameter() {defaultValue=3, parameterName=WidthBase}
  n2 = ScalarParameter() {defaultValue=7, parameterName=WidthTip}
  n3 = TextureCoordinate()
  n4 = WorldPosition()
  n5 = ActorPositionWS()
  n6 = Subtract(A=Divide(A=Subtract(A=@n4.XYZ, B=@n5)) {constB=15}, B=Multiply(A=ScalarParameter() {defaultValue=5.5, parameterName=CurlSpeed}, B=Time()))
  n7 = ScalarParameter() {defaultValue=1.75, parameterName=CurlScaleBot}
  n8 = Multiply(
   A=ScalarParameter() {defaultValue=5, parameterName=CurlIntensity}
   B=VectorNoise(World Position=Divide(A=@n6, B=@n7)) {noiseFunction=VNF_VectorALU}
  )
  n9 = Constant()
  n10 = LinearInterpolate(
   A=VectorParameter() {parameterName=MaxOffsetPos, defaultValue=<rgba 0.05 0.05 0.05 1>}.RGB
   B=VectorParameter() {parameterName=MaxOffsetNeg, defaultValue=<rgba -0.05 -0.05 -0.05 1>}.RGB
  )
  n11 = Subtract(A=WorldPosition().XYZ, B=ActorPositionWS())
  n12 = DynamicParameter() {defaultValue=<rgba 0.25 1 -1 1>, paramNames [OpacityPanner, PanWorldOffsetMult, CenterGrad, Param4]}
  n13 = Multiply(A=@n12.PanWorldOffsetMult) {constB=0.15}
  n14 = ScalarParameter() {parameterName=MiddleGrad}
  n15 = Constant()
  n16 = SphereMask(
   A=Add(A=TextureCoordinate() {vTiling=0}, B=AppendVector(A=@n14, B=@n15))
   B=Constant2Vector() {r=0.5}
   Radius=ScalarParameter() {defaultValue=0.45, parameterName=MaskRad_1}
   Hardness=ScalarParameter() {defaultValue=0.3, parameterName=MaskHardness_1}
  )
  Add(
   A=Add(
    A=MaterialFunctionCall(WidthBase=@n1, WidthTip=@n2, UVs for Projection=@n3, UVs for Thickness=@n3, UVs for texturing=@n3) {materialFunction=&SplineThicken}.WorldPositionOffset
    B=ShadingPathSwitch(Default=@n8, Mobile=@n9) {desc=Migrated from FeatureLevelSwitch}
   )
   B=Multiply(A=Multiply(A=@n10, B=@n11), B=LinearInterpolate(A=@n13, B=@n12.PanWorldOffsetMult, Alpha=@n16))
  )
  ComponentMask(None=TextureCoordinate() {uTiling=2}) {r=true}
  n17 = Time()
  n18 = Abs(None=Add(A=Sine(None=@n17) {period=0.3285}, B=Sine(None=@n17) {period=0.4329}))
  Floor(None=@n18)
  n19 = Time()
  n20 = Abs(None=Add(A=Sine(None=@n19) {period=2.235}, B=Sine(None=@n19) {period=4.3}))
  Floor(None=@n20)
  n21 = Multiply(A=VectorParameter() {parameterName=TipColor, defaultValue=<rgba 0 0.6159 1 1>}.RGB, B=ParticleColor().G)
  n22 = Multiply(A=VectorParameter() {parameterName=BackColor, defaultValue=<rgba 0.0206 0.5219 1 1>}.RGB, B=ParticleColor().G)
  n23 = ComponentMask(None=TextureCoordinate() {uTiling=2}) {r=true}
  n24 = ComponentMask(None=TextureCoordinate() {vTiling=2}) {g=true}
  n25 = ScalarParameter() {parameterName=Map Color Gradient Along V}
  n26 = Clamp(None=Power(Base=LinearInterpolate(A=@n23, B=@n24, Alpha=@n25)) {constExponent=5})
  n27 = ParticleColor()
  n28 = VectorParameter() {parameterName=MidColor_Color, defaultValue=<rgba 0.0611 1 0.7272 1>}
  n29 = TextureCoordinate()
  n30 = Add(
   A=LinearInterpolate(A=ComponentMask(None=@n29) {r=true}, B=ComponentMask(None=@n29) {g=true}, Alpha=ScalarParameter() {parameterName=Map Color Gradient Along V})
   B=ScalarParameter() {parameterName=ColorChange_Offset}
  )
  n31 = OneMinus(None=@n30)
  n32 = Power(Base=Multiply(A=Multiply(A=@n30, B=@n31), B=ScalarParameter() {defaultValue=5, parameterName=TriColorBlend}) {constB=2}) {constExponent=5}
  MaterialFunctionCall(Emissive=LinearInterpolate(A=LinearInterpolate(A=@n21, B=@n22, Alpha=@n26), B=Multiply(A=@n27.G, B=@n28.RGB), Alpha=Clamp(None=@n32))) {materialFunction=&MF_FX_EmissiveAdjustments}
  n33 = Subtract(A=WorldPosition().XYZ, B=ActorPositionWS())
  n34 = Multiply(
   A=ScalarParameter() {defaultValue=250, parameterName=CurlSpeed_1}
   B=Multiply(A=Ceil(None=@n18), B=ScalarParameter() {defaultValue=5, parameterName=TwitchyNoise_Speed})
  )
  n35 = Subtract(A=@n33, B=@n34)
  n36 = Multiply(
   A=ScalarParameter() {defaultValue=2, parameterName=CurlScaleBot_1}
   B=Multiply(A=Ceil(None=@n20), B=ScalarParameter() {defaultValue=5, parameterName=TwitchyNoise_Scale})
  )
  n37 = ScalarParameter() {defaultValue=1, parameterName=TwitchyNoise_Speed_1}
  n38 = Time()
  n39 = Sine(None=@n38) {period=1.1234}
  n40 = Sine(None=@n38) {period=0.4329}
  n41 = Floor(None=Abs(None=Add(A=@n39, B=@n40)))
  Multiply(
   A=VectorNoise(World Position=Divide(A=@n35, B=@n36)) {noiseFunction=VNF_VectorALU}
   B=Multiply(A=ScalarParameter() {defaultValue=7, parameterName=CurlIntensity_1}, B=Multiply(A=@n37, B=@n41))
  )
  Multiply(A=@n33, B=@n34)
  n42 = Add(
   A=TextureCoordinate() {vTiling=0}
   B=AppendVector(
    A=DynamicParameter() {defaultValue=<rgba 0.25 1 -1 1>, paramNames [OpacityPanner, PanWorldOffsetMult, CenterGrad, Param4]}.CenterGrad
    B=Constant()
   )
  )
  n43 = Constant2Vector() {r=0.5}
  n44 = ScalarParameter() {defaultValue=0.5, parameterName=MaskRad}
  n45 = ScalarParameter() {defaultValue=0.3, parameterName=MaskHardness}
  n46 = VertexColor()
  n47 = Add(A=@n12.OpacityPanner) {constB=0.25}
  n48 = DepthFade(FadeDistance=ScalarParameter() {defaultValue=50, parameterName=Fade Distance})
  n49 = ScalarParameter() {parameterName=Allow Depth Fade}
  Multiply(
   A=Multiply(A=SphereMask(A=@n42, B=@n43, Radius=@n44, Hardness=@n45), B=Multiply(A=@n46.G, B=@n47))
   B=ShadingPathSwitch(Default=LinearInterpolate(B=@n48, Alpha=@n49) {constA=1}, Mobile=Constant() {r=1})
  )
 ]
}
