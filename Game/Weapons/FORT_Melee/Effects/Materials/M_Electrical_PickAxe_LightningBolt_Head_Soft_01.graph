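# Unlit translucent mesh-particle material for the soft head of an electrical pickaxe lightning bolt effect.
# A powered sine along U shapes the bolt, combining with U, particle alpha and Fresnel into depth-faded opacity.
# Color blends back to tip along U, tinted by particle color, with an offsettable parabolic band of pure particle color.
# That color times the sine mask, plus Fresnel particle color, feeds the emissive adjustment function.
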Material /Game/Weapons/FORT_Melee/Effects/Materials/M_Electrical_PickAxe_LightningBolt_Head_Soft_01
graph {
 properties {
  bUsedWithMeshParticles=true
  blendMode=BLEND_Translucent
  pixelDepthOffsetMode=PDOM_Legacy
  previewMesh=/Game/Weapons/FORT_Melee/Effects/Mesh/SM_ElectricalAxe_Head_01
  shadingModel=MSM_Unlit
  usedShadingModels=Unlit
 }
 nodes [
  n1 = TextureCoordinate() {uTiling=1.2}
  n2 = Power(Base=Sine(None=ComponentMask(None=@n1) {r=true}) {period=2}, Exp=ScalarParameter() {defaultValue=1, parameterName=TipColor_Power})
  n3 = ParticleColor()
  n4 = Multiply(A=ComponentMask(None=TextureCoordinate()) {r=true}, B=@n3.A)
  n5 = MaterialFunctionCall(Exponent=ScalarParameter() {defaultValue=3, parameterName=FresnelExponent}) {materialFunction=&MF_Fresnel_FX}
  DepthFade(
   Opacity=Multiply(A=Clamp(None=@n2), B=Multiply(A=@n4, B=@n5.Fresnel))
   FadeDistance=ScalarParameter() {defaultValue=25, parameterName=DepthFade}
  )
  n6 = VectorParameter() {parameterName=BackColor_Color, defaultValue=<rgba 1 0.5954 0.0899 1>}
  n7 = VectorParameter() {parameterName=TipColor_Color, defaultValue=<rgba 0.4704 0.0403 1 1>}
  n8 = ComponentMask(None=TextureCoordinate()) {r=true}
  n9 = Add(A=ComponentMask(None=TextureCoordinate()) {r=true}, B=ScalarParameter() {parameterName=ColorChange_Offset})
  n10 = Multiply(A=Multiply(A=@n9, B=OneMinus(None=@n9))) {constB=4}
  n11 = LinearInterpolate(A=Multiply(A=LinearInterpolate(A=@n6.RGB, B=@n7.RGB, Alpha=@n8), B=@n3.RGB), B=@n3.RGB, Alpha=Clamp(None=Power(Base=@n10)))
  MaterialFunctionCall(Emissive=Add(A=Multiply(A=@n11, B=@n2), B=Multiply(A=@n3.RGB, B=@n5.Fresnel))) {materialFunction=&MF_FX_EmissiveAdjustments}
 ]
}
