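# Masked unlit material for the glowing Tier 6 pickaxe melee weapon.
# Emissive adds dimmed diffuse to a tinted laser-sword glow texture, panned and warped by scrolling noise, with inverted Fresnel shaping both before an overbright multiply.
# CharacterResInEffect feeds Opacity, while LODEffect and BlackHoleFunction drive World Position Offset.
# The material attributes then pass through ElementalWeaponsFunction and the accessories tech-art overrides.
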Material /Game/Weapons/FORT_Melee/Materials/M_PickaxeTier6_Glowy
graph {
 properties {
  bUseMaterialAttributes=true
  bUsedWithMeshDeformer=true
  bUsedWithSkeletalMesh=true
  blendMode=BLEND_Masked
  pixelDepthOffsetMode=PDOM_Legacy
  shadingModel=MSM_Unlit
  usedShadingModels=Unlit
 }
 nodes [
  n1 = TextureSampleParameter2D() {parameterName=Diffuse, texture=/Game/Weapons/FORT_Melee/Textures/Pickaxe_Tier_6_D}
  n2 = MaterialFunctionCall(Exponent=Constant() {r=2.5}) {materialFunction=&MF_Fresnel_FX}
  n3 = LinearInterpolate(A=ScalarParameter() {defaultValue=0.3, parameterName=DiffStrength}, Alpha=@n2.Fresnel Inverted)
  n4 = TextureSample(UVs=Panner(Coordinate=TextureCoordinate() {coordinateIndex=1}) {speedX=0.1, speedY=0.21}) {texture=&T_NoisePack_01}
  n5 = MaterialFunctionCall(X=@n4.R, Y=@n4.R) {materialFunction=&MakeFloat2}
  n6 = Multiply(A=Subtract(A=Multiply(A=@n5.Result) {constB=2})) {constB=0.1}
  n7 = Panner(Coordinate=TextureCoordinate() {coordinateIndex=1, vTiling=1.5}) {bFractionalPart=true, speedY=0.6}
  n8 = Multiply(
   A=TextureSample(UVs=Add(A=@n6, B=@n7)) {texture=/Game/Weapons/FORT_Melee/Effects/LaserSwordFounders/T_LaserSwordGlow}.RGB
   B=VectorParameter() {parameterName=GlowColour, defaultValue=<rgba 0.1168 0.5365 1.5 0>}.RGB
  )
  n9 = Multiply(
   A=Add(A=Multiply(A=@n1.RGB, B=@n3) {constB=0.6}, B=Multiply(A=@n8, B=@n2.Fresnel Inverted))
   B=ScalarParameter() {defaultValue=2, parameterName=OverbrightFactor}
  )
  n10 = MaterialFunctionCall!(Animation Gradient, Rez In Percentage) {materialFunction=/Game/Packages/Fortress_SharedMaterials/Base_Material_Functions/CharacterResInEffect}
  n11 = TextureSampleParameter2D() {parameterName=Normals, samplerType=SAMPLERTYPE_Normal, texture=/Game/Weapons/FORT_Melee/Textures/Pickaxe_Tier_6_N}
  n12 = VectorParameter() {group=WorldPositionOffset, parameterName=WorldPostionOfBlackHole, defaultValue=<rgba -30000 25536 1952 0>}
  n13 = MaterialFunctionCall!(
   World Position
   DistanceCullFadeAnimationOffset=MaterialFunctionCall() {materialFunction=/Game/Packages/Fortress_SharedMaterials/Base_Material_Functions/PickupEffectDistanceCullOverride}.Result
   Use actor center instead of pivot point
  ) {materialFunction=/Game/Packages/Fortress_SharedMaterials/Base_Material_Functions/LODEffect}
  n14 = WorldPosition()
  n15 = Add(
   A=MaterialFunctionCall!(WorldPostionOfBlackHole=AppendVector(A=@n12.RGB, B=@n12.RGB), World Position=Add(A=@n13.Result, B=@n14.XYZ)) {materialFunction=/Game/Packages/Fortnite_UniqueMaterials_1/BlackHoleFunction}.Result
   B=@n13.Result
  )
  MaterialFunctionCall(In=MaterialFunctionCall!(
    In=MakeMaterialAttributes!(
     BaseColor
     Metallic
     Specular
     Roughness
     Anisotropy
     EmissiveColor=@n9
     Opacity=@n10.Modulative Opacity
     OpacityMask
     Normal=@n11.RGB
     Tangent
     WorldPositionOffset=@n15
     SubsurfaceColor
     ClearCoat
     ClearCoatRoughness
     AmbientOcclusion
     Refraction
     CustomizedUVs_0
     CustomizedUVs_1
     CustomizedUVs_2
     CustomizedUVs_3
     CustomizedUVs_4
     CustomizedUVs_5
     CustomizedUVs_6
     CustomizedUVs_7
     PixelDepthOffset
     ShadingModel
     Displacement
    )
    Pre-Skinned Local Position
    Pre-Skinned Local Normal
   ) {materialFunction=/Game/Weapons/WeaponSkins/MaterialFunction/ElementalWeaponsFunction}.Result) {materialFunction=&MF_TechArtOverrides_accessories}
 ]
}
