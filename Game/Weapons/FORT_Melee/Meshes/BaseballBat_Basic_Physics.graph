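# Physics asset for the basic baseball bat mesh among the melee weapons, giving it a simple physics collision shape.
# A single body on the Melee_TwoHanded_Handle bone uses one long box, offset and rotated along the bone.
# It has no constraints, so the bat is represented as one rigid piece with no articulated parts.
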PhysicsAsset /Game/Weapons/FORT_Melee/Meshes/BaseballBat_Basic_Physics
bodies {Melee_TwoHanded_Handle [{shapeType=Box, extentX=10.5732, extentY=10.4521, extentZ=102.4016, center=<0.7405 -0.2115 31.3079>, rotation=<rot 0.1792 -90 0.1786>}]}
constraints []
