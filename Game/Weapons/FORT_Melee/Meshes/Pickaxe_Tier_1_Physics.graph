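# Physics asset for the tier 1 melee pickaxe mesh, giving it simplified collision for physics simulation.
# A single box body is bound to the Melee_TwoHanded_Handle bone, offset and rotated to cover the tool.
# It has no constraints, so the pickaxe acts as one rigid piece.
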PhysicsAsset /Game/Weapons/FORT_Melee/Meshes/Pickaxe_Tier_1_Physics
bodies {Melee_TwoHanded_Handle [{shapeType=Box, extentX=69.8425, extentY=16.6398, extentZ=103.9618, center=<-0.4998 -4.862 27.5133>, rotation=<rot 1.6612 -90 -0.1136>}]}
constraints []
