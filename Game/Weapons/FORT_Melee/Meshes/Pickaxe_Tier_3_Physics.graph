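# Physics asset for the tier 3 pickaxe melee weapon mesh, giving it a simple collision shape.
# It holds a single body on the Melee_TwoHanded_Handle bone, made of one rotated, offset box.
# With no constraints, the pickaxe simulates and collides as one rigid piece.
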PhysicsAsset /Game/Weapons/FORT_Melee/Meshes/Pickaxe_Tier_3_Physics
bodies {Melee_TwoHanded_Handle [{shapeType=Box, extentX=61.5281, extentY=20.527, extentZ=120.5476, center=<0.3096 -4.197 28.2897>, rotation=<rot 2.8462 -90 0.5325>}]}
constraints []
