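# Physics asset for the tier 4 pickaxe melee weapon mesh, giving it simple collision for physics simulation and queries.
# It holds a single box body on the Melee_TwoHanded_Handle bone, offset and rotated to enclose the weapon as one shape.
# It has no constraints, so the weapon acts as one rigid piece.
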PhysicsAsset /Game/Weapons/FORT_Melee/Meshes/Pickaxe_Tier_4_Physics
bodies {Melee_TwoHanded_Handle [{shapeType=Box, extentX=70.0346, extentY=18.3446, extentZ=125.4956, center=<1.0723 -4.7025 31.0425>, rotation=<rot 0.9264 -90 0.0291>}]}
constraints []
