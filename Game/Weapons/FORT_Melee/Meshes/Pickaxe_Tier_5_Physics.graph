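# Physics asset for the tier 5 melee pickaxe mesh, giving the weapon a simplified collision and physics shape.
# It holds a single body on the Melee_TwoHanded_Handle bone, built from one rotated, offset box.
# It defines no constraints, so the pickaxe simulates as one rigid piece.
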PhysicsAsset /Game/Weapons/FORT_Melee/Meshes/Pickaxe_Tier_5_Physics
bodies {Melee_TwoHanded_Handle [{shapeType=Box, extentX=73.3392, extentY=16.8576, extentZ=134.2333, center=<0.8652 1.947 34.5189>, rotation=<rot -0.0867 -90 1.3042>}]}
constraints []
