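# Physics asset for the tier 6 pickaxe melee weapon mesh, defining its collision shape for physics simulation.
# A single body on the Melee_TwoHanded_Handle bone uses one rotated box offset along the handle.
# It has no constraints, so the pickaxe behaves as one rigid piece.
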PhysicsAsset /Game/Weapons/FORT_Melee/Meshes/Pickaxe_Tier_6_Physics
bodies {Melee_TwoHanded_Handle [{shapeType=Box, extentX=71.5773, extentY=14.5579, extentZ=122.4252, center=<0.6761 -0.091 30.5413>, rotation=<rot -0.3046 -90 0.1289>}]}
constraints []
