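# Physics asset for the Tier 7 pickaxe melee weapon mesh, giving it a simple collision shape for physics.
# A single body on the Melee_TwoHanded_Handle bone uses one rotated, offset box to cover the weapon.
# It has no constraints, so the pickaxe simulates as one rigid piece.
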PhysicsAsset /Game/Weapons/FORT_Melee/Meshes/Pickaxe_Tier_7_Physics
bodies {Melee_TwoHanded_Handle [{shapeType=Box, extentX=64.4634, extentY=29.5826, extentZ=125.8355, center=<0.25 7.3052 32.0445>, rotation=<rot -2.0319 -90 -0.009>}]}
constraints []
