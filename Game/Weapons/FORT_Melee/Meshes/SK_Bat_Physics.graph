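# Physics asset for the two-handed bat melee weapon skeletal mesh, SK_Bat.
# A single box body on the Melee_TwoHanded_Handle bone forms a tall, slightly tilted collision volume offset along the bat's length.
# With no constraints, the whole weapon acts as one rigid body in physics.
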PhysicsAsset /Game/Weapons/FORT_Melee/Meshes/SK_Bat_Physics
bodies {Melee_TwoHanded_Handle [{shapeType=Box, extentX=23.6939, extentY=21.5139, extentZ=106.4594, center=<2.1491 -2.5655 31.8682>, rotation=<rot -2.1871 -90 2.9218>}]}
constraints []
