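# Physics asset for the basic fire axe two-handed melee weapon skeletal mesh.
# It holds a single rigid body on the Melee_TwoHanded_Handle bone, shaped as one rotated, offset box.
# No constraints exist, so the weapon simulates as one solid piece.
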PhysicsAsset /Game/Weapons/FORT_Melee/Meshes/SK_FireAxe_Basic_Physics
bodies {Melee_TwoHanded_Handle [{shapeType=Box, extentX=50.7414, extentY=10.9323, extentZ=105.636, center=<0.7923 5.552 22.5338>, rotation=<rot -2.0647 -90 -0.5206>}]}
constraints []
