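# Physics asset for the Flintlock Scythe melee weapon skeletal mesh, used for its collision.
# A single body on the Melee_TwoHanded_Handle bone holds one offset box shape, thin on one axis and long along Z, as a rough bound for the weapon.
# It has no constraints, so the weapon simulates as one rigid piece with no jointed parts.
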PhysicsAsset /Game/Weapons/FORT_Melee/Meshes/SK_Flintlock_Scythe_Physics
bodies {Melee_TwoHanded_Handle [{shapeType=Box, extentX=7.9861, extentY=58.6491, extentZ=155.9577, center=<0.4537 11.2254 38.4581>}]}
constraints []
