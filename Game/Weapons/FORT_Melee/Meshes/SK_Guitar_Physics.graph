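# Physics asset for the SK_Guitar two-handed melee weapon mesh, giving it a simplified collision shape.
# A single box body on the Melee_TwoHanded_Handle bone, offset and rotated relative to that bone, is the only collision shape.
# There are no constraints, so the asset defines one rigid body with no jointed parts.
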PhysicsAsset /Game/Weapons/FORT_Melee/Meshes/SK_Guitar_Physics
bodies {Melee_TwoHanded_Handle [{shapeType=Box, extentX=47.0948, extentY=12.5521, extentZ=123.457, center=<-1.9444 -0.7661 36.6489>, rotation=<rot 0.4317 -90 -0.9259>}]}
constraints []
