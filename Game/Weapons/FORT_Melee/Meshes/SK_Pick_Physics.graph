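# Physics asset for the SK_Pick skeletal mesh, a melee pickaxe weapon under FORT_Melee.
# It holds a single box body on the Melee_TwoHanded_Handle bone, slightly offset and rotated to enclose the weapon.
# It has no constraints, so the pick simulates and collides as one rigid piece.
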PhysicsAsset /Game/Weapons/FORT_Melee/Meshes/SK_Pick_Physics
bodies {Melee_TwoHanded_Handle [{shapeType=Box, extentX=62.2482, extentY=26.5578, extentZ=107.443, center=<0.4942 -0.1113 23.4261>, rotation=<rot -0.7335 -90 0.7302>}]}
constraints []
