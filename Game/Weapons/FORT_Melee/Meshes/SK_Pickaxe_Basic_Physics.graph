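# Physics asset for the basic two-handed melee pickaxe skeletal mesh.
# It holds a single body on the Melee_TwoHanded_Handle bone, shaped as one rotated, offset box that stands in for the pickaxe's collision.
# No constraints are defined, so the pickaxe simulates as one rigid piece rather than an articulated ragdoll.
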PhysicsAsset /Game/Weapons/FORT_Melee/Meshes/SK_Pickaxe_Basic_Physics
bodies {Melee_TwoHanded_Handle [{shapeType=Box, extentX=68.593, extentY=11.7791, extentZ=120.1066, center=<0.9535 -3.6893 25.6838>, rotation=<rot -1.2211 -90 0.0675>}]}
constraints []
