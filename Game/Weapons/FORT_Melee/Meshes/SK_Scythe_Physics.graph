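# Physics asset for the two-handed scythe skeletal mesh among the melee weapons.
# It holds a single body on the Melee_TwoHanded_Handle bone, shaped as one rotated, offset box that is long and thin.
# There are no constraints, so the weapon collides as one rigid piece with no articulated parts.
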PhysicsAsset /Game/Weapons/FORT_Melee/Meshes/SK_Scythe_Physics
bodies {Melee_TwoHanded_Handle [{shapeType=Box, extentX=58.0122, extentY=4.5306, extentZ=130.8398, center=<0.1348 8.6033 27.2236>, rotation=<rot 6.3434 -90 0.0066>}]}
constraints []
