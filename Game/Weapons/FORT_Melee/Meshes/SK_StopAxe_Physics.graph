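# Physics asset for the StopAxe two-handed melee weapon skeletal mesh.
# It has a single body on the Melee_TwoHanded_Handle bone, built from one box shape that is elongated along its local Z and offset from the bone origin.
# There are no constraints, so the weapon behaves as one rigid piece for collision and physics.
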PhysicsAsset /Game/Weapons/FORT_Melee/Meshes/SK_StopAxe_Physics
bodies {Melee_TwoHanded_Handle [{shapeType=Box, extentX=55.99, extentY=9.664, extentZ=153.2534, center=<-0.1149 -0.8449 44.7851>, rotation=<rot -0.1659 -90 -0.0094>}]}
constraints []
