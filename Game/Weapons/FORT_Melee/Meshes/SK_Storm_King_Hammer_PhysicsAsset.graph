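# Physics asset for the Storm King Hammer two-handed melee weapon skeletal mesh.
# It holds a single body on the Melee_TwoHanded_Handle bone, shaped as one capsule offset and rotated to lie along the weapon.
# There are no constraints, so it has no jointed or articulated physics parts.
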PhysicsAsset /Game/Weapons/FORT_Melee/Meshes/SK_Storm_King_Hammer_PhysicsAsset
bodies {Melee_TwoHanded_Handle [{shapeType=Capsule, length=133.0358, radius=13.2182, center=<3.11 1.8 34.97>, rotation=<rot 1.201 -90 -0.689>}]}
constraints []
