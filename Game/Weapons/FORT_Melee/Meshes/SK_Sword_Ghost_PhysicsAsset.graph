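# Physics asset for the Ghost sword skeletal mesh, a melee weapon, giving it a simple collision and physics shape.
# It holds a single body on the Melee_TwoHanded_Handle bone, made of one long capsule offset along and aligned with the weapon.
# There are no constraints, so the sword acts as one rigid piece with no articulated parts.
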PhysicsAsset /Game/Weapons/FORT_Melee/Meshes/SK_Sword_Ghost_PhysicsAsset
bodies {Melee_TwoHanded_Handle [{shapeType=Capsule, length=119.7132, radius=15.8309, center=<-0.0047 0.0017 52.3695>, rotation=<rot 0.0091 -90 -0.044>}]}
constraints []
